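# Niagara skydiving contrail cosmetic with four elemental trails styled as water, fire, air and earth, attached to skeletal mesh sockets.
# Merged_Trails places one particle group per element, bending ribbons into swirling paths from dive, lean and front-end velocity inputs.
# Sparks, flame sprites and rock pieces use Particle Attribute Readers to follow those trail points.
# Socket-attached airflow meshes, glows and lights add accents, with glow size and light brightness rising during dives.
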NiagaraSystem /BRCosmetics/Athena/Effects/Particles/Skydiving/NS_Contrail_PlantStand/NS_Contrail_PlantStand
dependencies {
 usedDataInterfaces [
  &NiagaraDataInterfacePlatformSet
  &NiagaraDataInterfaceArrayFloat
  &NiagaraDataInterfaceParticleRead
  &NiagaraDataInterfaceCurve
  &NiagaraDataInterfaceVector2DCurve
  &NiagaraDataInterfaceArrayColor
  &NiagaraDataInterfaceArrayFloat3
  &NiagaraDataInterfaceSkeletalMesh
 ]
 usedDynamicInputs [
  &MakeVector2D
  &Lerp_Vector
  &SetBoolByFloatComparison
  &DotProduct
  &TransformVector
 ]
 usedModules [
  &self:SystemSpawnScript.SystemScriptSource.SystemScriptGraph.NiagaraNodeAssignment_2.SetVariables_#1
  &self:SystemSpawnScript.SystemScriptSource.SystemScriptGraph.NiagaraNodeAssignment_1.SetVariables_#2
  &SystemState
  /Game/Effects/Niagara/Modules/Skydiving/FN_DivingAttributes/v4/FN_DivingAttributes_v4
  &self:SystemSpawnScript.SystemScriptSource.SystemScriptGraph.NiagaraNodeAssignment_0.SetVariables_#3
  &Timeline
 ]
}
emitters {
 Air_Airflow {
  emitterUpdateScript [
   &EmitterState(Life Cycle Mode=::NewEnumerator1, Loop Behavior=::NewEnumerator1, Loop Delay=0.03, Loop Duration Mode=::NewEnumerator1, UseLoopDelay=false)
   &SpawnRate(Spawn Probability=0.5, SpawnRate=10, Use Spawn Probability=true)
  ]
  particleSpawnScript [
   &InitializeParticle(
    Color Maximum=<rgba 0.4 0.47 0.5 0.4>
    Color Minimum=<rgba 0.8 0.9367 1 0.6>
    Color Mode=::NewEnumerator2
    Lifetime Max=0.8
    Lifetime Min=0.5
    Lifetime Mode=::NewEnumerator1
    Mesh Scale=&Multiply_VectorByFloat(
     Vector=&RandomRangeVector(Minimum=<1 1 1.5>, Maximum=<1 1 2>)
     Float=&RandomRangeFloat(Minimum=0.25, Maximum=0.35)
    )
    Mesh Scale Mode=::NewEnumerator4
    UsePositionOffset=false
    Write Material Random=true
   )
   &InitialMeshOrientation(
    Mesh Orientation Mode=::NewEnumerator2
    Orientation Coordinate Space=::Local
    Rotation=&Multiply_VectorByFloat(Vector=&RandomRangeVector(Minimum=<-1 -1 -1>, Maximum=<1 1 1>), Float=1)
    Use Rotation Vector=true
   )
   &CameraOffset(Camera Offset Amount=5)
  ]
  particleUpdateScript [
   &ParticleState()
   &DynamicMaterialParameters(
    Index 0 Param 1=&FloatFromCurve(
     FloatCurve=DataInterface() {propertyValues {Curve {keys [{interpMode=RCIM_Cubic, tangentMode=RCTM_User, time=0, value=1}, {interpMode=RCIM_Cubic, tangentMode=RCTM_User, time=1, value=0}]}}}
     Scale Curve=0.8
    )
    Module.Param0WriteEnabled=true
    Module.Param1WriteEnabled=true
    Module.Param2WriteEnabled=true
    Module.Param3WriteEnabled=true
   )
   &SkeletalMeshLocation(
    Bone / Socket Index=2
    Bone Sampling Mode=::NewEnumerator4
    MaskSpawnGroup=false
    Orientation Sampling=::NewEnumerator1
    Skeletal Mesh=$System.Skeletal Mesh
   )
   &ScaleMeshSize(Scale Factor=&Multiply_VectorByFloat(
     Vector=<1 1 1>
     Float=&FloatFromCurve(FloatCurve=DataInterface() {propertyValues {Curve {keys [
          {arriveTangent=1.0107, interpMode=RCIM_Cubic, leaveTangent=1.0107, tangentMode=RCTM_User, time=0, value=0.5}
          {interpMode=RCIM_Cubic, tangentMode=RCTM_User, time=1, value=1}
         ]}}})
    ))
   &UpdateMeshOrientation(Rotation Rate=&RandomRangeFloat(Minimum=-2, Maximum=2), Rotation Vector=<0 0 1>)
   &SolveForcesAndVelocity(Clamp Velocity=false, Limit Acceleration=false)
   &ScaleColor(
    Scale Alpha=&FloatFromCurve(
     FloatCurve=DataInterface() {propertyValues {Curve {keys [
         {arriveTangent=4.2786, interpMode=RCIM_Cubic, leaveTangent=4.2786, tangentMode=RCTM_User, time=0, value=0}
         {interpMode=RCIM_Cubic, tangentMode=RCTM_User, time=0.3, value=1}
         {tangentMode=RCTM_User, time=1, value=0}
        ]}}}
     Scale Curve=&RandomRangeFloat(Minimum=0.5, Maximum=0.75)
    )
    ScaleA=true
    ScaleRGB=true
   )
  ]
  rendererClasses [&NiagaraMeshRendererProperties]
  renderers [{rendererClass=&NiagaraMeshRendererProperties, rendererIndex=0}]
 }
 Fire_Sparks {
  emitterUpdateScript [
   &EmitterState(Life Cycle Mode=::NewEnumerator1, Loop Behavior=::NewEnumerator1, Loop Delay=2, Loop Duration Mode=::NewEnumerator1, UseLoopDelay=true)
   &SpawnRate(Spawn Probability=0.8, SpawnRate=59, Use Spawn Probability=true)
   &SpawnRate(Spawn Probability=0.8, SpawnRate=120, Use Spawn Probability=true) {module=SpawnRate001}
  ]
  particleSpawnScript [
   &InitializeParticle(
    Color Maximum=<rgba 1 0.0833 0 1>
    Color Minimum=<rgba 1.5 0.2965 0 1>
    Color Mode=::NewEnumerator2
    Lifetime=&RandomRangeFloat(Minimum=0.3, Maximum=0.6)
    Sprite Size Max=<2 8>
    Sprite Size Min=<1 6>
    Sprite Size Mode=::NewEnumerator2
    UsePositionOffset=false
    Write Material Random=true
   )
   &self:ScratchModule_05(
    New Particle Attribute Reader=DataInterface() {propertyValues {EmitterBinding {bindingMode=Other, emitterName=Merged_Trails}}}
    Particle Index:NiagaraInt32=&Add_Integer(
     A=&RandomRangeInt(
      Minimum=0
      Maximum=&MakeIntFromFloat(Float=&Multiply_FloatByInt(Float=0.6, Integer=$System.NumPoints))
     )
     B=&Multiply_Int(A=$System.NumPoints, B=1)
    )
   ) {module=ScratchModule_05}
   /SetVariables(Particles.RandomCore:NiagaraFloat=&FloatFromCurve(
     FloatCurve=DataInterface() {propertyValues {Curve {keys [
         {arriveTangent=0.0559, interpMode=RCIM_Cubic, leaveTangent=0.0559, tangentMode=RCTM_User, time=0, value=0.2}
         {arriveTangent=0.1895, interpMode=RCIM_Cubic, leaveTangent=0.1895, tangentMode=RCTM_User, time=0.9, value=0.25}
         {arriveTangent=1.1808, interpMode=RCIM_Cubic, leaveTangent=1.1808, tangentMode=RCTM_User, time=1, value=0.3}
        ]}}}
     CurveIndex=&RandomRangeFloat(Minimum=0, Maximum=1)
     Scale Curve=2
    )) {isSetParameters=true, module=SetVariables_#4}
   &ShapeLocation(MaskSpawnGroup=false, Override Seed=false, Sphere Radius=5, UseOwnerScale=true)
   &AddVelocity(
    Cone Angle=10
    Cone Axis=$Output.ScratchModule_05.Direction Vector
    Module.UseConeFalloff=false
    Module.UseVelDistribution=false
    Override Seed=false
    Velocity Mode=::NewEnumerator2
    Velocity Speed=&RandomRangeFloat(Minimum=100, Maximum=300)
   )
  ]
  particleUpdateScript [
   &ParticleState()
   &CurlNoiseForce(Noise Strength=500, Pan Noise=true, Pan Noise Field=<0 0 0.1>)
   &ScaleSpriteSize(Scale Factor=&Vector2DFromFloat(Value=&FloatFromCurve(
      FloatCurve=DataInterface() {propertyValues {Curve {keys [
          {arriveTangent=12.3508, interpMode=RCIM_Cubic, leaveTangent=12.3508, tangentMode=RCTM_User, time=0, value=0}
          <at 0.2 1.2 RCIM_Cubic>
          {arriveTangent=-2.7132, interpMode=RCIM_Cubic, leaveTangent=-2.7132, tangentMode=RCTM_User, time=1, value=0}
         ]}}}
      Scale Curve=0.5
     )))
   /SetVariables(Particles.Glow03_Size=&Multiply_Vector2D(A=$Particles.SpriteSize, B=<4 3>)) {isSetParameters=true}
   &ScaleColor(
    Scale Alpha=&FloatFromCurve(FloatCurve=DataInterface() {propertyValues {Curve {keys [
         {arriveTangent=9.0725, interpMode=RCIM_Cubic, leaveTangent=9.0725, tangentMode=RCTM_User, time=0, value=0}
         {arriveTangent=-2.5249, interpMode=RCIM_Cubic, leaveTangent=-2.5249, tangentMode=RCTM_User, time=0.1408, value=0.9847}
         {arriveTangent=-0.0405, interpMode=RCIM_Cubic, leaveTangent=-0.0405, tangentMode=RCTM_User, time=1, value=0}
        ]}}})
    Scale RGB=&Multiply_VectorByFloat(Vector=<2 2 2>, Float=&FNTimeOfDayScalar(Night Scale=1, Morning Scale=0.7, Evening Scale=0.7))
    ScaleA=true
    ScaleRGB=true
   )
   /SetVariables(Particles.Glow03_Color=&MultiplyLinearColor(
     Linear Color=$Particles.Color
     Scale Factor=&MakeFloat4FromVectorAndFloat("Vector (RGB)"=<1.5 1.5 1.5>, "Float (Alpha)"=0.1)
    )) {isSetParameters=true}
   &Drag(Drag=8, Use Linear Drag=true, Use Rotational Drag=false)
   &SolveForcesAndVelocity(Clamp Velocity=false, Limit Acceleration=false)
  ]
  rendererClasses [&NiagaraSpriteRendererProperties]
  renderers [
   {rendererClass=&NiagaraSpriteRendererProperties}
   {rendererClass=&NiagaraSpriteRendererProperties, rendererIndex=1}
  ]
 }
 FlameSprites {
  emitterUpdateScript [
   &EmitterState(Life Cycle Mode=::NewEnumerator1, Loop Behavior=::NewEnumerator1, Loop Delay=1, Loop Duration Mode=::NewEnumerator1, UseLoopDelay=true)
   &SpawnRate(Spawn Probability=0.8, SpawnRate=25, Use Spawn Probability=false)
   &SpawnRate(Spawn Probability=0.85, SpawnRate=19, Use Spawn Probability=true) {module=SpawnRate001}
  ]
  particleSpawnScript [
   &InitializeParticle(
    Color Maximum=<rgba 0.5 0.05 0 0.3>
    Color Minimum=<rgba 2 0.3333 0 0.5>
    Color Mode=::NewEnumerator2
    Lifetime Max=1
    Lifetime Min=0.6
    Lifetime Mode=::NewEnumerator1
    Sprite Rotation Mode=::NewEnumerator2
    Sprite Size Mode=::NewEnumerator1
    Sprite UV Mode=::NewEnumerator3
    Uniform Sprite Size Max=50
    Uniform Sprite Size Min=30
    UsePositionOffset=false
    Write Material Random=true
   ) {module=InitializeParticle001}
   &self:ScratchModule_06(
    New Particle Attribute Reader=DataInterface() {propertyValues {EmitterBinding {bindingMode=Other, emitterName=Merged_Trails}}}
    Particle Index:NiagaraInt32=&Add_Integer(
     A=&RandomRangeInt(
      Minimum=0
      Maximum=&MakeIntFromFloat(Float=&Multiply_FloatByInt(Float=0.6, Integer=$System.NumPoints))
     )
     B=&Multiply_Int(A=$System.NumPoints, B=1)
    )
   ) {module=ScratchModule_06}
   &CameraOffset(Camera Offset Amount=5)
   &AddVelocity(
    Velocity=$Output.ScratchModule_06.Direction Vector
    Velocity Speed Scale=&RandomRangeFloat(Minimum=0, Maximum=20)
   )
  ]
  particleUpdateScript [
   /SetVariables(Particles.SubImageIndex=&FloatFromCurve(
     FloatCurve=DataInterface() {propertyValues {Curve {keys [{interpMode=RCIM_Cubic, tangentMode=RCTM_User, time=0, value=0}, {interpMode=RCIM_Cubic, tangentMode=RCTM_User, time=1, value=1}]}}}
     Scale Curve=16
    )) {isSetParameters=true}
   &ParticleState()
   &SkeletalMeshLocation(
    Bone / Socket Index=1
    Bone Sampling Mode=::NewEnumerator4
    MaskSpawnGroup=false
    Orientation Sampling=::NewEnumerator1
    Position Sampling=::NewEnumerator1
    Skeletal Mesh=$System.Skeletal Mesh
    Write To Particle Position=false
   ) {module=SkeletalMeshLocation001}
   &ScaleSpriteSize(Scale Factor=&Vector2DFromCurve(
     Vector2Curve=DataInterface() {propertyValues {
       XCurve {keys [
         {arriveTangent=1.9071, interpMode=RCIM_Cubic, leaveTangent=1.9071, tangentMode=RCTM_User, time=0, value=0.425}
         {interpMode=RCIM_Cubic, tangentMode=RCTM_User, time=0.3, value=1}
         {arriveTangent=-1.7158, tangentMode=RCTM_User, time=1, value=0.05}
        ]}
       YCurve {keys [
         {arriveTangent=1.9071, interpMode=RCIM_Cubic, leaveTangent=1.9071, tangentMode=RCTM_User, time=0, value=0.425}
         {interpMode=RCIM_Cubic, tangentMode=RCTM_User, time=0.3, value=1}
         {arriveTangent=-1.7158, tangentMode=RCTM_User, time=1, value=0.05}
        ]}
      }}
     ScaleCurve=<1 1>
    ))
   &ScaleColor(
    Scale Alpha=&Multiply_Float(
     A=&FloatFromCurve(
      FloatCurve=DataInterface() {propertyValues {Curve {keys [
          {interpMode=RCIM_Cubic, tangentMode=RCTM_User, time=0, value=0}
          {interpMode=RCIM_Cubic, tangentMode=RCTM_User, time=0.5, value=1}
          {tangentMode=RCTM_User, time=1, value=0}
         ]}}}
      Scale Curve=&FN_FrontEndFloatModifier(Gameplay=1, Front End=0.2)
     )
     B=1
    )
    Scale RGB=&VectorFromFloat(Value=0.8)
    ScaleA=true
    ScaleRGB=true
   ) {module=ScaleColor001}
   &Drag(Drag=5, Use Linear Drag=true, Use Rotational Drag=true)
   &CurlNoiseForce(Noise Frequency=10, Noise Strength=50, Pan Noise=true, Pan Noise Field=<0 0 0.1>)
   &SolveForcesAndVelocity(Clamp Velocity=false, Limit Acceleration=false)
  ]
  rendererClasses [&NiagaraSpriteRendererProperties]
  renderers [{rendererClass=&NiagaraSpriteRendererProperties}]
 }
 Light_and_Glow {
  emitterUpdateScript [
   &EmitterState(Life Cycle Mode=::NewEnumerator1, Loop Behavior=::NewEnumerator1, Loop Duration=1, UseLoopDelay=false)
   &SpawnBurst_Instantaneous(Spawn Count=4, Spawn Probability=HlslExpression() {hlslExpression="Emitter.LoopCount == 0"}, Use Spawn Probability=true) {module=SpawnBurst_Instantaneous001}
  ]
  particleSpawnScript [
   &FN_SetSpawnGroup()
   &InitializeParticle(
    Color=&SelectLinearColorFromArray(
     Array Sampling Mode=::NewEnumerator1
     Direct Array Index=&SpawnGroup()
     Color Selection Array=DataInterface() {propertyValues {ColorData [<rgba 0.15 0.8298 1.5 1>, <rgba 1.5 0.25 0 1>, <rgba 0.6588 0.8222 1 1>, <rgba 1 0.7083 0.5 1>]}}
     Random Seed=?
    )
    Lifetime=3
    Sprite Size Mode=::NewEnumerator3
    Uniform Sprite Size=25
    UsePositionOffset=false
    Write Material Random=true
   ) {module=InitializeParticle001}
  ]
  particleUpdateScript [
   &ParticleState(Kill Particles When Lifetime Has Elapsed=false)
   &SkeletalMeshLocation(
    Bone / Socket Index=$Particles.SpawnGroup
    Bone Sampling Mode=::NewEnumerator4
    Inherit Velocity=false
    Interpolate Mesh Sampling=false
    MaskSpawnGroup=false
    Sampled Position Offset=&SelectVectorFromArray(
     Direct Array Index=&SpawnGroup()
     Vector Selection Array=DataInterface() {propertyValues {FloatData [<0 0 0>, <0 1 1>, <5 -7 0>, <-5 7 0>]}}
    )
    Sampled Position Offset Coordinate Space:ENiagaraMeshTransforms=::NewEnumerator3
    Skeletal Mesh=$System.Skeletal Mesh
   )
   &ScaleColor(
    Scale Alpha=&FloatFromCurve(
     FloatCurve=DataInterface() {propertyValues {Curve {keys [{interpMode=RCIM_Cubic, tangentMode=RCTM_User, time=0, value=0}, {interpMode=RCIM_Cubic, tangentMode=RCTM_User, time=1, value=1}]}}}
     Scale Curve=0.5
    )
    ScaleA=true
    ScaleRGB=true
   )
   /SetVariables(Particles.LightColor=&MultiplyLinearColor(
     Linear Color=$Particles.Color
     Scale Factor=&MakeVector4_FromFloat(Float=&FloatFromCurve(FloatCurve=DataInterface() {propertyValues {Curve {keys [<at 0 1>,<at 1 2>]}}}, CurveIndex=$System.Dive))
    )) {isSetParameters=true}
   /SetVariables(Particles.SpriteSize=&Multiply_Vector2D_ByFloat(Float=&Lerp_Float(A=1, B=2, Alpha=$System.Dive), Vector2D=$Particles.Initial.SpriteSize)) {isSetParameters=true}
  ]
  rendererClasses [&NiagaraLightRendererProperties, &NiagaraSpriteRendererProperties]
  renderers [
   {rendererClass=&NiagaraLightRendererProperties, rendererIndex=0}
   {rendererClass=&NiagaraSpriteRendererProperties, rendererIndex=1}
  ]
 }
 Merged_Trails {
  emitterUpdateScript [
   &EmitterState(Life Cycle Mode=::NewEnumerator1, Loop Behavior=::NewEnumerator1, Loop Duration=1, UseLoopDelay=false)
   &SpawnBurst_Instantaneous(
    Spawn Count=&Multiply_Int(A=$System.NumPoints, B=4)
    Spawn Probability=HlslExpression() {hlslExpression="Emitter.LoopCount == 0"}
    Use Spawn Probability=false
   ) {module=SpawnBurst_Instantaneous001}
  ]
  particleSpawnScript [
   &FN_SetSpawnGroup(
    Inclusivity=true
    Normalize Spawn Group=true
    Particles Per Group:NiagaraInt32=$System.NumPoints
    Pattern=::NewEnumerator0
    Set Ribbon ID and Ribbon Link Order=true
   )
   &InitializeParticle(
    Color=&MaskLinearColorBySpawnGroup(
     Masked Linear Color=<rgba 0 0.6452 2 1>
     Passthrough Linear Color=&MaskLinearColorBySpawnGroup(
      Masked Linear Color=<rgba 2 0.1667 0 0.5>
      Passthrough Linear Color=&MaskLinearColorBySpawnGroup(Masked Linear Color=<rgba 0.8054 0.9497 1 0.3>, Passthrough Linear Color=<rgba 0.2 0.16 0.12 1.3>, Spawn Group Mask=2)
      Spawn Group Mask=1
     )
    )
    Lifetime=&MaskFloatBySpawnGroup(
     Masked Float=2
     Passthrough Float=&MaskFloatBySpawnGroup(
      Masked Float=2.3
      Passthrough Float=&MaskFloatBySpawnGroup(Masked Float=2, Passthrough Float=1.8, Spawn Group Mask=2)
      Spawn Group Mask=1
     )
    )
    Mesh Scale Mode=::NewEnumerator3
    Ribbon Width=&MaskFloatBySpawnGroup(
     Masked Float=30
     Passthrough Float=&MaskFloatBySpawnGroup(
      Masked Float=30
      Passthrough Float=&MaskFloatBySpawnGroup(Masked Float=25, Passthrough Float=30, Spawn Group Mask=2)
      Spawn Group Mask=1
     )
    )
    Ribbon Width Mode=::NewEnumerator1
    UsePositionOffset=false
    Write Material Random=true
   ) {module=InitializeParticle001}
   /SetVariables(Particles.Ribbon Length=DataInterface() {propertyValues {GpuSyncMode=SyncCpuToGpu, MaxElements=0, FloatData [160,180,170,165]}}) {isSetParameters=true}
   &DynamicMaterialParameters(
    Index 0 Param 3=&MaskFloatBySpawnGroup(
     Masked Float=50
     Passthrough Float=&MaskFloatBySpawnGroup(
      Masked Float=100
      Passthrough Float=&MaskFloatBySpawnGroup(Masked Float=20, Passthrough Float=50, Spawn Group Mask=2)
      Spawn Group Mask=1
     )
    )
    Index 0 Param 4=&MaskFloatBySpawnGroup(
     Masked Float=0
     Passthrough Float=&MaskFloatBySpawnGroup(
      Masked Float=0.1
      Passthrough Float=&MaskFloatBySpawnGroup(Masked Float=0.63, Passthrough Float=0.63, Spawn Group Mask=2)
      Spawn Group Mask=1
     )
    )
    Module.Index0Param0Write=true
    Module.Index0Param1Write=true
    Module.Index0Param2Write=true
    Module.Index0Param3Write=true
   )
   /SetVariables(Particles.Linear Color=&FN_Select_Color_from_Time_Of_Day(Morning=<rgba 1 1 1 1>, Day=<rgba 1 1 1 1>, Evening=<rgba 1 1 1 1>, Night=<rgba 0.75 0.75 0.75 1>)) {isSetParameters=true}
  ]
  particleUpdateScript [
   &ParticleState(Kill Particles When Lifetime Has Elapsed=false)
   &SkeletalMeshLocation(
    Bone / Socket Index=&ModuloInt(A=&SpawnGroup(), B=4)
    Bone Sampling Mode=::NewEnumerator4
    MaskSpawnGroup=false
    Orientation Sampling=::NewEnumerator1
    Skeletal Mesh=$System.Skeletal Mesh
    Use Sampled Mesh Attribute as Ribbon ID=false
   )
   &SkeletalMeshLocation(
    Bone / Socket Index=4
    Bone Sampling Mode=::NewEnumerator4
    MaskSpawnGroup=false
    Orientation Sampling=::NewEnumerator1
    Position Sampling=::NewEnumerator2
    Skeletal Mesh=$System.Skeletal Mesh
    Use Sampled Mesh Attribute as Ribbon ID=false
   ) {module=SkeletalMeshLocation001}
   /SetVariables(Particles.GlideLeanQuat:Quat4f=&LerpQuaternion(
     Quaternion A:Quat4f=&MultiplyQuaternion(
      Quaternion A=$Particles.SkeletalMeshLocation001.SampledRotation
      Quaternion B=&MakeQuatFromAxisAngle(Axis=<0 0 1>, AngleInDegrees=-15)
     )
     Quaternion B:Quat4f=&MultiplyQuaternion(
      Quaternion A=&MultiplyQuaternion(
       Quaternion A=&MultiplyQuaternion(
        Quaternion A=$Particles.SkeletalMeshLocation001.SampledRotation
        Quaternion B=&MakeQuatFromAxisAngle(Axis=<0 0 1>, AngleInDegrees=-15)
       )
       Quaternion B=&MultiplyQuaternion(
        Quaternion A=&MakeQuatFromAxisAngle(Axis=<1 0 0>, AngleInDegrees=&Lerp_Float(A=0, B=-50, Alpha=$System.Timeline.Time))
        Quaternion B=&MakeQuatFromAxisAngle(Axis=<1 0 0>, AngleInDegrees=&Lerp_Float(A=0, B=50, Alpha=$System.Timeline001.Time))
       )
      )
      Quaternion B=&MakeQuatFromAxisAngle(
       Axis=<0 0 1>
       AngleInDegrees=&Lerp_Float(
        A=0
        B=-60
        Alpha=&RemapRange(Input Value To Remap=$System.Dive, Input Min=0.28, Input Max=1, Output Min=0, Output Max=1)
       )
      )
     )
     Lerp Factor:NiagaraFloat=&FloatFromCurve(
      FloatCurve=DataInterface() {propertyValues {Curve {keys [
          {arriveTangent=2.5236, interpMode=RCIM_Cubic, leaveTangent=2.5236, tangentMode=RCTM_User, time=0, value=0}
          {interpMode=RCIM_Cubic, tangentMode=RCTM_User, time=0.9, value=1}
         ]}}}
      CurveIndex=$Particles.Normalized Spawn Group
     )
     Enable Spherical Linear Interpolation:NiagaraBool=false
    )) {isSetParameters=true}
   /SetVariables(Particles.Position=&AddVectorToPosition(
     Position=$Particles.Position
     Vector=&Multiply_VectorByFloat(
      Vector=&Add_Vector(
       A=&NormalizeVector(
        Vector To Normalize=&FN_FrontEndVector3Modifier(
         Gameplay=&RotateVectorByQuaternion(Vector To Rotate=<0 -1 0>, Quaternion=$Particles.GlideLeanQuat)
         Front End=$System.FrontendVelocity
        )
        Safe Normalize=true
        Safe Fallback Threshold=0
        Safe Fallback Vector=<0 0 0.001>
       )
       B=<0 0 0>
      )
      Float=&FloatFromCurve(
       FloatCurve=DataInterface() {propertyValues {Curve {keys [
           {arriveTangent=1.3593, interpMode=RCIM_Cubic, leaveTangent=1.3593, tangentMode=RCTM_User, time=0, value=0}
           {arriveTangent=0.6459, tangentMode=RCTM_User, time=1, value=1}
          ]}}}
       CurveIndex=$Particles.Normalized Spawn Group
       Scale Curve=&SelectFloatFromArray(Direct Array Index=&SpawnGroup(), Float Selection Array=$Particles.Ribbon Length)
      )
     )
    )) {isSetParameters=true}
   &RotateAroundPoint(
    Radius:NiagaraFloat=&FloatFromCurve(
     FloatCurve=DataInterface() {propertyValues {Curve {keys [
         {arriveTangent=-0.9921, interpMode=RCIM_Cubic, leaveTangent=-0.9921, tangentMode=RCTM_User, time=0, value=1}
         {arriveTangent=-0.4865, interpMode=RCIM_Cubic, leaveTangent=-0.4865, tangentMode=RCTM_User, time=1, value=0.25}
        ]}}}
     CurveIndex=$Particles.Normalized Spawn Group
     Scale Curve=80
    )
    Rotation Center:NiagaraPosition=$Particles.SkeletalMeshLocation001.SampledPosition
    Rotation Phase:NiagaraFloat=&Add_Float(
     A=&Add_Float(
      A=&SelectFloatFromArray(
       Direct Array Index=&SpawnGroup()
       Float Selection Array=DataInterface() {propertyValues {FloatData [0.5,-0.3,0.2,0]}}
      )
      B=-0.1
     )
     B=&Add_Float(
      A=&Multiply_Float(A=$Particles.Normalized Spawn Group, B=2)
      B=&Divide_Float(A=&MakeFloatFromInt(Int=$Particles.SpawnGroup), B=4)
     )
    )
    Set Particle Position=false
    X Axis:Vector3f=&RotateVectorByQuaternion(Vector To Rotate=<1 0 0>, Quaternion=$Particles.GlideLeanQuat)
    Y Axis:Vector3f=&RotateVectorByQuaternion(Vector To Rotate=<0 0 1>, Quaternion=$Particles.GlideLeanQuat)
   )
   /SetVariables(Particles.Position=&LerpPosition(
     A=$Particles.Position
     B=&AddVectorToPosition(
      Position=$Output.RotateAroundPoint.OutputPosition
      Vector=&Multiply_VectorByFloat(
       Vector=&NormalizeVector(
        Vector To Normalize=&FN_FrontEndVector3Modifier(
         Gameplay=&RotateVectorByQuaternion(Vector To Rotate=<0 -1 0>, Quaternion=$Particles.GlideLeanQuat)
         Front End=$System.FrontendVelocity
        )
        Safe Normalize=true
        Safe Fallback Threshold=0
        Safe Fallback Vector=<0 0 0.001>
       )
       Float=&FloatFromCurve(
        FloatCurve=DataInterface() {propertyValues {Curve {keys [
            {arriveTangent=0.0308, interpMode=RCIM_Cubic, leaveTangent=0.0308, tangentMode=RCTM_User, time=0, value=0}
            {arriveTangent=0.8636, interpMode=RCIM_Cubic, leaveTangent=0.8636, tangentMode=RCTM_User, time=1, value=1}
           ]}}}
        CurveIndex=$Particles.Normalized Spawn Group
        Scale Curve=&SelectFloatFromArray(Direct Array Index=&SpawnGroup(), Float Selection Array=$Particles.Ribbon Length)
       )
      )
     )
     Alpha=&FloatFromCurve(
      FloatCurve=DataInterface() {propertyValues {Curve {keys [
          {arriveTangent=3.2474, interpMode=RCIM_Cubic, leaveTangent=3.2474, tangentMode=RCTM_User, time=0, value=0}
          {arriveTangent=0.0224, interpMode=RCIM_Cubic, leaveTangent=0.0224, tangentMode=RCTM_User, time=1, value=1}
         ]}}}
      CurveIndex=$Particles.Normalized Spawn Group
     )
    )) {isSetParameters=true}
   &SkeletalMeshLocation(
    Bone / Socket Index=&ModuloInt(A=&SpawnGroup(), B=4)
    Bone Sampling Mode=::NewEnumerator4
    MaskSpawnGroup=false
    Orientation Sampling=::NewEnumerator1
    Position Sampling=::NewEnumerator2
    Sampled Position Offset=/Niagara/DynamicInputs/Vectors/RotateVector(
     Vector To Rotate:Vector3f=&Multiply_VectorByFloat(
      Vector=&NormalizeVector(Vector To Normalize=<0 0 1>, Safe Normalize=true, Safe Fallback Threshold=0, Safe Fallback Vector=<0 0 0.001>)
      Float=&FloatFromCurve(
       FloatCurve=DataInterface() {propertyValues {Curve {keys [
           {arriveTangent=1.741, interpMode=RCIM_Cubic, leaveTangent=1.741, tangentMode=RCTM_User, time=0, value=0.4}
           {interpMode=RCIM_Cubic, tangentMode=RCTM_User, time=0.6, value=1}
          ]}}}
       CurveIndex=$Particles.Normalized Spawn Group
       Scale Curve=&SelectFloatFromArray(
        Direct Array Index=&SpawnGroup()
        Float Selection Array=DataInterface() {propertyValues {FloatData [20,20,22,22]}}
       )
      )
     )
     Yaw:NiagaraFloat=&FN_FrontEndFloatModifier(
      Gameplay=-0.02
      Front End=&Lerp_Float(
       A=-0.02
       B=&SelectFloatFromArray(
        Direct Array Index=&SpawnGroup()
        Float Selection Array=DataInterface() {propertyValues {FloatData [0.2,0.2,-0.02,-0.02]}}
       )
       Alpha=$User.ParaGlideLeanAlpha
      )
     )
     Pitch:NiagaraFloat=&Lerp_Float(A=0, B=0, Alpha=$User.ParaGlideLeanAlpha)
     Roll:NiagaraFloat=&Add_Float(
      A=$Particles.Normalized Spawn Group
      B=&SelectFloatFromArray(
       Direct Array Index=&SpawnGroup()
       Float Selection Array=DataInterface() {propertyValues {FloatData [0.1,-0.1,-0.1,0]}}
      )
     )
     Rotation Rate:NiagaraFloat=&SelectFloatFromArray(
      Direct Array Index=&SpawnGroup()
      Float Selection Array=DataInterface() {propertyValues {FloatData [1000,1000,1000,1000]}}
     )
    )
    Skeletal Mesh=$System.Skeletal Mesh
    Use Sampled Mesh Attribute as Ribbon ID=false
   ) {module=SkeletalMeshLocation002}
   /SetVariables(Particles.Position=&LerpPosition(
     A=$Particles.SkeletalMeshLocation002.SampledPosition
     B=$Particles.Position
     Alpha=&FloatFromCurve(
      FloatCurve=DataInterface() {propertyValues {Curve {keys [
          {arriveTangent=-0.0137, interpMode=RCIM_Cubic, leaveTangent=-0.0137, tangentMode=RCTM_User, time=0.45, value=0}
          {arriveTangent=0.0505, interpMode=RCIM_Cubic, leaveTangent=0.0505, tangentMode=RCTM_User, time=0.6, value=1}
         ]}}}
      CurveIndex=$Particles.Normalized Spawn Group
     )
    )) {isSetParameters=true}
   /SetVariables() {isSetParameters=true}
   &self:EvalCurve(
    New Particle Attribute Reader=DataInterface() {propertyValues {EmitterBinding {bindingMode=Self, emitterName=None}}}
    Particle Index:NiagaraInt32=&Add_Integer(
     A=&MakeIntFromFloat(Float=&FloatFromCurve(
       FloatCurve=DataInterface() {propertyValues {Curve {keys [
           {arriveTangent=2.2022, interpMode=RCIM_Cubic, leaveTangent=2.2022, tangentMode=RCTM_User, time=0, value=0}
           {interpMode=RCIM_Cubic, tangentMode=RCTM_User, time=1, value=1}
          ]}}}
       CurveIndex=$Emitter.Age
       Scale Curve=&MakeFloatFromInt(Int=&ReturnExecIndex())
      ))
     B=&Multiply_Int(A=$System.NumPoints, B=$Emitter.SpawnGroup)
    )
   ) {module=EvalCurve}
   &ScaleColor(
    Scale Alpha=&MaskFloatBySpawnGroup(
     Masked Float=0.1
     Passthrough Float=&MaskFloatBySpawnGroup(
      Masked Float=1
      Passthrough Float=&MaskFloatBySpawnGroup(Masked Float=0.5, Passthrough Float=1, Spawn Group Mask=2)
      Spawn Group Mask=1
     )
    )
    Scale RGB=&MaskVectorBySpawnGroup(
     Masked Vector=&FN_Select_Vector_from_Time_Of_Day(
      Morning=&MakeVectorFromLinearColorRGB(Color=<rgba 1 1 1 1>)
      Day=&MakeVectorFromLinearColorRGB(Color=&MakeCustomLinearColorFromBool(
        Bool=&Boolean_OrOperation(Bool1=$User.bIsFrontEnd, Bool2=$User.bIsFrontEndPreview)
        True Color=<rgba 0.9 0.8475 0.81 1>
        False Color=<rgba 1 1 1 1>
       ))
      Evening=&MakeVectorFromLinearColorRGB(Color=<rgba 1 1 1 1>)
      Night=&MakeVectorFromLinearColorRGB(Color=<rgba 0.81 0.8453 0.9 1>)
     )
     Passthrough Vector=<1 1 1>
     Spawn Group Mask=2
    )
    ScaleA=true
    ScaleRGB=true
   )
   &ScaleRibbonWidth(Ribbon Width Scale=&MaskFloatBySpawnGroup(
     Masked Float=&FloatFromCurve(
      FloatCurve=DataInterface() {propertyValues {Curve {keys [
          {arriveTangent=0.1674, interpMode=RCIM_Cubic, leaveTangent=0.1674, tangentMode=RCTM_User, time=0, value=0}
          {arriveTangent=4.7687, interpMode=RCIM_Cubic, leaveTangent=0.0238, tangentMode=RCTM_Break, time=0.3, value=1}
          {arriveTangent=-2.8418, interpMode=RCIM_Cubic, leaveTangent=-2.8418, tangentMode=RCTM_User, time=1, value=0}
         ]}}}
      CurveIndex=$Particles.Normalized Spawn Group
     )
     Passthrough Float=&MaskFloatBySpawnGroup(
      Masked Float=&FloatFromCurve(
       FloatCurve=DataInterface() {propertyValues {Curve {keys [
           {arriveTangent=2.7388, interpMode=RCIM_Cubic, leaveTangent=2.7388, tangentMode=RCTM_User, time=0, value=0}
           {arriveTangent=2.0533, interpMode=RCIM_Cubic, leaveTangent=2.0533, tangentMode=RCTM_User, time=0.4, value=0.85}
           {arriveTangent=0.0151, interpMode=RCIM_Cubic, leaveTangent=0.0151, tangentMode=RCTM_User, time=0.7, value=0.85}
           {arriveTangent=0.0104, interpMode=RCIM_Cubic, leaveTangent=0.0104, tangentMode=RCTM_User, time=1, value=1}
          ]}}}
       CurveIndex=$Particles.Normalized Spawn Group
       Scale Curve=&FloatFromCurve(
        FloatCurve=DataInterface() {propertyValues {Curve {keys [
            {interpMode=RCIM_Cubic, tangentMode=RCTM_User, time=0, value=0.8}
            {interpMode=RCIM_Cubic, tangentMode=RCTM_User, time=0.3, value=1.2}
            {tangentMode=RCTM_User, time=1, value=0.8}
           ]}}}
        CurveIndex=$Emitter.NormalizedLoopAge
       )
      )
      Passthrough Float=&MaskFloatBySpawnGroup(
       Masked Float=&FloatFromCurve(
        FloatCurve=DataInterface() {propertyValues {Curve {keys [
            {arriveTangent=0.9586, interpMode=RCIM_Cubic, leaveTangent=0.9586, tangentMode=RCTM_User, time=0, value=0}
            {arriveTangent=0.0238, interpMode=RCIM_Cubic, leaveTangent=0.0238, tangentMode=RCTM_User, time=0.7, value=1.2}
            {arriveTangent=-2.8418, interpMode=RCIM_Cubic, leaveTangent=-2.8418, tangentMode=RCTM_User, time=1, value=0}
           ]}}}
        CurveIndex=$Particles.Normalized Spawn Group
       )
       Passthrough Float=&FloatFromCurve(
        FloatCurve=DataInterface() {propertyValues {Curve {keys [
            {arriveTangent=0.9586, interpMode=RCIM_Cubic, leaveTangent=0.9586, tangentMode=RCTM_User, time=0, value=0}
            {arriveTangent=0.8676, interpMode=RCIM_Cubic, leaveTangent=0.8676, tangentMode=RCTM_User, time=0.7, value=1}
            {arriveTangent=-2.8418, interpMode=RCIM_Cubic, leaveTangent=-2.8418, tangentMode=RCTM_User, time=1, value=0}
           ]}}}
        CurveIndex=$Particles.Normalized Spawn Group
       )
       Spawn Group Mask=2
      )
      Spawn Group Mask=1
     )
    ))
   /SetVariables(
    Particles.RibbonWidth_A:NiagaraFloat=&MaskFloatBySpawnGroup(Masked Float=$Particles.RibbonWidth, Passthrough Float=0)
    Particles.RibbonWidth_B:NiagaraFloat=&MaskFloatBySpawnGroup(Masked Float=$Particles.RibbonWidth, Passthrough Float=0, Spawn Group Mask=1)
    Particles.RibbonWidth_C:NiagaraFloat=&MaskFloatBySpawnGroup(Masked Float=$Particles.RibbonWidth, Passthrough Float=0, Spawn Group Mask=2)
    Particles.RibbonWidth_D:NiagaraFloat=&MaskFloatBySpawnGroup(Masked Float=$Particles.RibbonWidth, Passthrough Float=0, Spawn Group Mask=3)
   ) {isSetParameters=true}
  ]
  rendererClasses [&NiagaraMeshRendererProperties, &NiagaraRibbonRendererProperties]
  renderers [
   {rendererClass=&NiagaraMeshRendererProperties, rendererIndex=0}
   {rendererClass=&NiagaraRibbonRendererProperties, rendererIndex=1}
   {rendererClass=&NiagaraRibbonRendererProperties, rendererIndex=2}
   {rendererClass=&NiagaraRibbonRendererProperties, rendererIndex=3}
   {rendererClass=&NiagaraRibbonRendererProperties, rendererIndex=4}
   {rendererClass=&NiagaraRibbonRendererProperties, rendererIndex=5}
   {rendererClass=&NiagaraRibbonRendererProperties, rendererIndex=6}
   {rendererClass=&NiagaraRibbonRendererProperties, rendererIndex=7}
  ]
 }
 RockPieces {
  emitterSpawnScript [
   /Game/Effects/Niagara/MasterMaterials/ToonShader/Template/NM_ToonShaderParameters(HighLight_Size=0.1636)
   /Game/Effects/Niagara/MasterMaterials/ToonShader/Template/NM_OutlineParameters(OutlineWidth:NiagaraFloat=&FN_PickFloatDependOnPlatform(TargetPlatformActive=1, TargetPlatformNoActive=0.4))
   /SetVariables(Emitter.Switch Outline Thickness=1) {isSetParameters=true}
  ]
  emitterUpdateScript [
   &EmitterState(Life Cycle Mode=::NewEnumerator1, Loop Duration=1, UseLoopDelay=false)
   &SpawnRate(SpawnRate=6, Use Spawn Probability=false)
   &SpawnRate(Spawn Probability=0.7, SpawnRate=5, Use Spawn Probability=true) {module=SpawnRate001}
  ]
  particleSpawnScript [
   &InitializeParticle(
    Color=&RandomRangeLinearColor(Minimum=<rgba 0.6 0.475 0.3 1>, Maximum=<rgba 0.35 0.2319 0.1925 1>)
    Lifetime Max=7
    Lifetime Min=6
    Lifetime Mode=::NewEnumerator1
    Mass Mode=::NewEnumerator1
    Mesh Scale Mode=::NewEnumerator3
    Mesh Uniform Scale=&SelectFloatFromArray(
     Array Sampling Mode=::NewEnumerator0
     Direct Array Index=?
     Float Selection Array=DataInterface() {propertyValues {FloatData [0.75,1,1.2,2]}}
     Random Seed=0
    )
    UsePositionOffset=false
    Write Material Random=true
   )
   /SetVariables(Particles.MeshRandom:NiagaraInt32=&RandomRangeInt(Minimum=0, Maximum=2)) {isSetParameters=true}
   &InitialMeshOrientation(Use Rotation Vector=false)
   &ShapeLocation(MaskSpawnGroup=false, Override Seed=false, Sphere Radius=3, UseOwnerScale=true)
   &CameraOffset(Camera Offset Amount=3)
  ]
  particleUpdateScript [
   &ParticleState()
   &self:ScratchModule_04(
    New Particle Attribute Reader=DataInterface() {propertyValues {EmitterBinding {bindingMode=Other, emitterName=Merged_Trails}}}
    Particle Index:NiagaraInt32=&Add_Integer(
     A=&MakeIntFromFloat(
      Float=&FloatFromCurve(
       FloatCurve=DataInterface() {propertyValues {Curve {keys [
           {arriveTangent=1.3565, interpMode=RCIM_Cubic, leaveTangent=1.3565, tangentMode=RCTM_User, time=0, value=0}
           {arriveTangent=0.3647, interpMode=RCIM_Cubic, leaveTangent=0.3647, tangentMode=RCTM_User, time=0.5, value=0.5}
           {arriveTangent=0.9797, interpMode=RCIM_Cubic, leaveTangent=0.9797, tangentMode=RCTM_User, time=1, value=1}
          ]}}}
       Scale Curve=&MakeFloatFromInt(Int=$System.NumPoints)
      )
      Conversion Mode=::NewEnumerator1
     )
     B=&Multiply_Int(A=$System.NumPoints, B=3)
    )
   ) {module=ScratchModule_04}
   &InterpolateOverTime(Rate Of Change=6, Target Vector Value:NiagaraPosition=$Particles.Position, Variable Type=::NewEnumerator2)
   /SetVariables(Particles.Position=&AddVectorToPosition(Position=$Particles.InterpolateOverTime.InterpolateOverTime_V3_.Moving Average, Vector=$Particles.Initial.ShapeLocation.ShapeVector)) {isSetParameters=true}
   &ScaleMeshSize(Scale Factor=&VectorFromFloat(Value=&FloatFromCurve(
      FloatCurve=DataInterface() {propertyValues {Curve {keys [
          {arriveTangent=0.8365, interpMode=RCIM_Cubic, leaveTangent=0.8365, tangentMode=RCTM_User, time=0, value=0}
          {interpMode=RCIM_Cubic, tangentMode=RCTM_User, time=0.6, value=1}
          {arriveTangent=-5.0767, tangentMode=RCTM_User, time=1, value=0}
         ]}}}
      Scale Curve=&RandomRangeFloat(Minimum=0.3, Maximum=0.7)
     )))
   &UpdateMeshOrientation(
    Rotation Rate=&RandomRangeFloat(Minimum=-1, Maximum=1)
    Rotation Vector=&RandomVector(Vector Scale=1)
   )
   &ScaleColor(
    Scale RGB=&FN_Select_Vector_from_Time_Of_Day(
     Morning=&MakeVectorFromLinearColorRGB(Color=<rgba 1 1 1 1>)
     Day=&MakeVectorFromLinearColorRGB(Color=&MakeCustomLinearColorFromBool(
       Bool=&Boolean_OrOperation(Bool1=$User.bIsFrontEnd, Bool2=$User.bIsFrontEndPreview)
       True Color=<rgba 1 1 1 1>
       False Color=<rgba 1 1 1 1>
      ))
     Evening=&MakeVectorFromLinearColorRGB(Color=<rgba 1 1 1 1>)
     Night=&MakeVectorFromLinearColorRGB(Color=<rgba 1.1 0.9856 0.88 1>)
    )
    ScaleA=true
    ScaleRGB=true
   )
  ]
  rendererClasses [&NiagaraMeshRendererProperties]
  renderers [
   {rendererClass=&NiagaraMeshRendererProperties, rendererIndex=0}
   {rendererClass=&NiagaraMeshRendererProperties, rendererIndex=1}
   {rendererClass=&NiagaraMeshRendererProperties, rendererIndex=2}
   {rendererClass=&NiagaraMeshRendererProperties, rendererIndex=3}
  ]
 }
 Water_Sparks {
  emitterUpdateScript [
   &EmitterState(Life Cycle Mode=::NewEnumerator1, Loop Behavior=::NewEnumerator1, Loop Delay=2, Loop Duration Mode=::NewEnumerator1, UseLoopDelay=true)
   &SpawnRate(Spawn Probability=0.8, SpawnRate=49, Use Spawn Probability=true)
   &SpawnRate(Spawn Probability=0.8, SpawnRate=100, Use Spawn Probability=true) {module=SpawnRate001}
  ]
  particleSpawnScript [
   &InitializeParticle(
    Color Maximum=<rgba 0.204 0.3632 1 1>
    Color Minimum=<rgba 0.05 0.2381 0.5 1>
    Color Mode=::NewEnumerator2
    Lifetime=&RandomRangeFloat(Minimum=0.2, Maximum=0.4)
    Sprite Size Max=<1.5 8>
    Sprite Size Min=<1 4>
    Sprite Size Mode=::NewEnumerator2
    UsePositionOffset=false
    Write Material Random=true
   )
   &SkeletalMeshLocation(
    Bone / Socket Index=2
    Bone Sampling Mode=::NewEnumerator4
    "Inherit Velocity (Scale)":NiagaraFloat=&RandomRangeFloat(Minimum=0.5, Maximum=1)
    Inherited Velocity Speed Limit=150
    MaskSpawnGroup=false
    Orientation Sampling=::NewEnumerator1
    Skeletal Mesh=$System.Skeletal Mesh
    Velocity Sampling=::NewEnumerator0
    VelocitySpeedLimit=true
   )
   /SetVariables(Particles.RandomCore:NiagaraFloat=&FloatFromCurve(
     FloatCurve=DataInterface() {propertyValues {Curve {keys [
         {arriveTangent=0.0559, interpMode=RCIM_Cubic, leaveTangent=0.0559, tangentMode=RCTM_User, time=0, value=0.2}
         {arriveTangent=0.1895, interpMode=RCIM_Cubic, leaveTangent=0.1895, tangentMode=RCTM_User, time=0.9, value=0.25}
         {arriveTangent=1.1808, interpMode=RCIM_Cubic, leaveTangent=1.1808, tangentMode=RCTM_User, time=1, value=0.3}
        ]}}}
     CurveIndex=&RandomRangeFloat(Minimum=0, Maximum=1)
     Scale Curve=2
    )) {isSetParameters=true, module=SetVariables_#4}
   &self:ScratchModule_07(
    New Particle Attribute Reader=DataInterface() {propertyValues {EmitterBinding {bindingMode=Other, emitterName=Merged_Trails}}}
    Particle Index:NiagaraInt32=&RandomRangeInt(
     Minimum=0
     Maximum=&MakeIntFromFloat(Float=&Multiply_FloatByInt(Float=0.8, Integer=$System.NumPoints))
    )
   ) {module=ScratchModule_07}
   &ShapeLocation(MaskSpawnGroup=false, Override Seed=false, Sphere Radius=5, UseOwnerScale=true)
   &AddVelocity(
    Cone Angle=10
    Cone Axis=$Output.ScratchModule_07.Direction Vector
    Module.UseConeFalloff=false
    Module.UseVelDistribution=false
    Override Seed=false
    Velocity Mode=::NewEnumerator2
    Velocity Speed=&RandomRangeFloat(Minimum=50, Maximum=100)
   )
  ]
  particleUpdateScript [
   &ParticleState()
   &CurlNoiseForce(Noise Frequency=10, Noise Strength=150, Pan Noise=true, Pan Noise Field=<0 0 1>)
   &ScaleSpriteSize(Scale Factor=&Vector2DFromFloat(Value=&FloatFromCurve(FloatCurve=DataInterface() {propertyValues {Curve {keys [
          {arriveTangent=9.5944, interpMode=RCIM_Cubic, leaveTangent=9.5944, tangentMode=RCTM_User, time=0, value=0}
          <at 0.2 1.2 RCIM_Cubic>
          {arriveTangent=-2.7132, interpMode=RCIM_Cubic, leaveTangent=-2.7132, tangentMode=RCTM_User, time=1, value=0}
         ]}}})))
   /SetVariables(Particles.Glow03_Size=&Multiply_Vector2D(A=$Particles.SpriteSize, B=<4 3>)) {isSetParameters=true}
   &ScaleColor(
    Scale RGB=&Multiply_VectorByFloat(Vector=<2 2 2>, Float=&FNTimeOfDayScalar(Night Scale=1, Morning Scale=0.7, Evening Scale=0.7))
    ScaleA=true
    ScaleRGB=true
   )
   /SetVariables(Particles.Glow03_Color=&MultiplyLinearColor(
     Linear Color=$Particles.Color
     Scale Factor=&MakeFloat4FromVectorAndFloat("Vector (RGB)"=<1.5 1.5 1.5>, "Float (Alpha)"=0.1)
    )) {isSetParameters=true}
   &Drag(Drag=4, Use Linear Drag=true, Use Rotational Drag=false)
   &SolveForcesAndVelocity(Clamp Velocity=false, Limit Acceleration=false)
  ]
  rendererClasses [&NiagaraSpriteRendererProperties]
  renderers [
   {rendererClass=&NiagaraSpriteRendererProperties}
   {rendererClass=&NiagaraSpriteRendererProperties, rendererIndex=1}
  ]
 }
}
summary {
 systemName=NS_Contrail_PlantStand
 emitters [
  {emitterName=RockPieces, rendererClasses [&NiagaraMeshRendererProperties]}
  {emitterName=Fire_Sparks, rendererClasses [&NiagaraSpriteRendererProperties]}
  {emitterName=FlameSprites, rendererClasses [&NiagaraSpriteRendererProperties]}
  {emitterName=Air_Airflow, rendererClasses [&NiagaraMeshRendererProperties]}
  {emitterName=Water_Sparks, rendererClasses [&NiagaraSpriteRendererProperties]}
  {
   emitterName=Light_and_Glow
   rendererClasses [&NiagaraLightRendererProperties, &NiagaraSpriteRendererProperties]
  }
  {emitterName=Merged_Trails, rendererClasses [&NiagaraMeshRendererProperties, &NiagaraRibbonRendererProperties]}
 ]
 userVariables [
  User.bIsFrontEnd {type=NiagaraBool}
  User.bIsFrontEndPreview {type=NiagaraBool}
  User.ParaGlideLeanAlpha {type=NiagaraFloat, defaultValue=0, description=""}
 ]
}
